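# Translucent lit holographic material for the Creative item spawner vending machine's slot, showing a configurable icon, an arrow by default, with animated scanlines.
# The icon texture is scaled about its center and offset, then blends between fill and icon colors with a brightness multiplier.
# Time-panned sine waves along V form the scanlines, which modulate a master opacity before a separately weighted icon term is added.
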Material /Game/Creative/Devices/ItemSpawnerVendingMachine/Materials/M_Vending_Machine_slot_holo_01
graph {
 properties {blendMode=BLEND_Translucent, pixelDepthOffsetMode=PDOM_Legacy, usedShadingModels=Default Lit}
 nodes [
  n1 = TextureCoordinate()
  n2 = AppendVector(
   A=ScalarParameter() {defaultValue=1, group=Icon, parameterName=Icon Scale U, sortPriority=10}
   B=ScalarParameter() {defaultValue=1, group=Icon, parameterName=Icon Scale V, sortPriority=10}
  )
  n3 = ScalarParameter() {group=Icon, parameterName=Icon Offset U}
  n4 = ScalarParameter() {group=Icon, parameterName=Icon Offset V}
  n5 = TextureSampleParameter2D(UVs=Add(
    A=MaterialFunctionCall(UVs=@n1, Texture Scale=@n2) {materialFunction=&ScaleUVsByCenter}.UVs
    B=AppendVector(A=@n3, B=@n4)
   )) {
   group=Icon
   parameterName=Icon Texture
   samplerSource=SSM_Clamp_WorldGroupSettings
   sortPriority=1
   texture=/Game/Creative/Devices/ItemSpawnerVendingMachine/Textures/T_CM_Vending_Machine_Modern_arrow_D
  }
  n6 = TextureCoordinate()
  n7 = Time()
  n8 = ScalarParameter() {defaultValue=0.02, group=Scanlines, parameterName=ScanlinePanSpeed}
  n9 = Multiply(
   A=ScalarParameter() {defaultValue=20, group=Scanlines, parameterName=ScanlineTile}
   B=Add(A=ComponentMask(None=@n6) {g=true}, B=Multiply(A=@n7, B=@n8))
  ) {constA=1}
  n10 = Multiply(A=ConstantBiasScale(None=Sine(None=@n9)))
  n11 = ScalarParameter() {defaultValue=0.6, group=Scanlines, parameterName=ScanlineIntensity}
  Add(
   A=Multiply(A=@n5.RGB, B=ScalarParameter() {group=Color, parameterName=Added Icon Opacity})
   B=Multiply(A=LinearInterpolate(B=@n10, Alpha=@n11) {constA=1}, B=ScalarParameter() {defaultValue=0.4, group=Color, parameterName=Master Opacity})
  )
  Multiply(
   A=LinearInterpolate(
    A=VectorParameter() {group=Color, parameterName=Fill Color, sortPriority=1, defaultValue=<rgba 0 0.3848 0.6562 1>}.RGB
    B=VectorParameter() {group=Color, parameterName=Icon Color, sortPriority=1, defaultValue=<rgba 1 1 1 1>}.RGB
    Alpha=@n5.RGB
   )
   B=ScalarParameter() {defaultValue=1, group=Color, parameterName=Brightness Multiplier}
  )
 ]
}
